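# Unlit translucent Niagara material for flowing energy trails on ribbons and mesh particles.
# Emissive sums panning caustics, rainbow detail and curve-atlas-shaped noise, with shared normal-map UV distortion, tinted by emissive and particle color.
# A noise-broken U-gradient dissolve, driven by dynamic parameters, adds a bright edge and cuts opacity, alongside an optional smoke mask and non-mobile depth fade.
# A panning distortion texture feeds a local-space transformed world position offset.
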Material /Game/Effects/Niagara/MasterMaterials/SplineThickenTrail/M_BodyEnergy_Trail_BoneFlow
graph {
 properties {
  bCastRayTracedShadows=false
  bUsedWithNiagaraMeshParticles=true
  bUsedWithNiagaraRibbons=true
  blendMode=BLEND_Translucent
  pixelDepthOffsetMode=PDOM_Legacy
  shadingModel=MSM_Unlit
  usedShadingModels=Unlit
 }
 nodes [
  Constant()
  n1 = VectorParameter() {group=Emissive, parameterName=Emissive HighLight UV, sortPriority=1, defaultValue=<rgba 0.5 0.25 -0.5 0>}
  n2 = AppendVector(A=@n1.R, B=@n1.R)
  n3 = AppendVector(
   A=ScalarParameter() {defaultValue=1, group=Noise, parameterName=Global Scale U}
   B=ScalarParameter() {defaultValue=1, group=Noise, parameterName=Global Scale V}
  )
  n4 = Panner(Coordinate=Multiply(A=TextureCoordinate(), B=Multiply(A=@n2, B=@n3)), Speed=AppendVector(A=@n1.B, B=@n1.B))
  n5 = Desaturation(
   None=VectorParameter() {group=Emissive, parameterName=Emissive Color, defaultValue=<rgba 1 1 1 1>}.RGB
   Fraction=ScalarParameter() {group=Emissive, parameterName=Desaturation Color}
  )
  n6 = Multiply(
   A=Multiply(
    A=TextureSample(UVs=@n4) {texture=&T_ShieldCaustics}.R
    B=ScalarParameter() {defaultValue=10, group=Emissive, parameterName=Emissive HighLight Intensity}
   )
   B=@n5
  )
  n7 = TextureCoordinate()
  n8 = VectorParameter() {group=Distortion, parameterName=Distortion UV, sortPriority=1, defaultValue=<rgba 1 0.5 0 0>}
  n9 = AppendVector(
   A=ScalarParameter() {defaultValue=1, group=Noise, parameterName=Global Scale U}
   B=ScalarParameter() {defaultValue=1, group=Noise, parameterName=Global Scale V}
  )
  n10 = Multiply(A=AppendVector(A=@n8.R, B=@n8.R), B=@n9)
  n11 = TextureSampleParameter2D(UVs=Panner(Coordinate=Multiply(A=@n7, B=@n10), Speed=AppendVector(A=@n8.B, B=@n8.B))) {
   group=Distortion
   parameterName="Distortion  Tex"
   samplerType=SAMPLERTYPE_Normal
   sortPriority=0
   texture=&T_Weap_Distort_N_02
  }
  n12 = Multiply(
   A=ComponentMask(None=@n11.RGB) {g=true, r=true}
   B=ScalarParameter() {defaultValue=0.2, group=Distortion, parameterName=Distortion Intensity, sortPriority=2}
  )
  n13 = VectorParameter() {group=Emissive, parameterName=Emissive Detail UV, sortPriority=1, defaultValue=<rgba 1 1 -0.2 0>}
  n14 = ScalarParameter() {defaultValue=1, group=Noise, parameterName=Global Scale U}
  n15 = ScalarParameter() {defaultValue=1, group=Noise, parameterName=Global Scale V}
  n16 = Multiply(A=TextureCoordinate(), B=Multiply(A=AppendVector(A=@n13.R, B=@n13.R), B=AppendVector(A=@n14, B=@n15)))
  n17 = AppendVector(A=@n13.B, B=@n13.B)
  n18 = TextureSample(UVs=Add(A=@n12, B=Panner(Coordinate=@n16, Speed=@n17))) {texture=&T_Tile_RainbowColor}
  n19 = ScalarParameter() {group=Emissive, parameterName=Emissive Detail Intensity}
  n20 = VectorParameter() {group=Noise, parameterName=Noise A UV, sortPriority=6, defaultValue=<rgba 0.5 4 -0.5 0>}
  n21 = Multiply(A=TextureCoordinate(), B=Multiply(A=AppendVector(A=@n20.R, B=@n20.R), B=@n9))
  n22 = AppendVector(A=@n20.B, B=@n20.B)
  n23 = TextureSampleParameter2D(UVs=Add(A=@n12, B=Panner(Coordinate=@n21, Speed=@n22))) {group=Noise, parameterName=Noise Tex, samplerType=SAMPLERTYPE_LinearColor, texture=/Game/Global/Textures/Default/Noise/T_Noise01}
  n24 = CurveAtlasRowParameter(CurveTime=ComponentMask(None=TextureCoordinate()) {g=true}) {
   atlas=/Game/Athena/LinearColorCurves/A_Atlas_Turtleneck_Glider
   curve=/Game/Accessories/FORT_Backpacks/Backpack_M_MED_Calavera/FX/Curve_Backpack_Calavera_02
   defaultValue=40
   group=Emissive
   parameterName=Curve Color
  }
  n25 = CurveAtlasRowParameter(CurveTime=ComponentMask(None=TextureCoordinate()) {r=true}) {
   atlas=&A_Atlas_Turtleneck_Glider
   curve=&Curve_Backpack_Calavera_02
   defaultValue=40
   group=Emissive
   parameterName=Curve Color
  }
  n26 = Multiply(
   A=LinearInterpolate(
    A=ScalarParameter() {defaultValue=0.2, group=Noise, parameterName=Noise A Dark}
    B=ScalarParameter() {defaultValue=1.5, group=Noise, parameterName=Noise A Light}
    Alpha=Power(Base=@n23.R)
   )
   B=Multiply(A=@n24.G, B=@n25.R)
  )
  n27 = ScalarParameter() {defaultValue=100, group=Dissolve, parameterName=Dissolve Edge Intensity}
  n28 = DynamicParameter() {defaultValue=<rgba 0 1 1 1>, paramNames [Random Offset, Spline Width Mult, Dissolve, Dissolve Edge Intensity]}
  n29 = TextureCoordinate()
  n30 = OneMinus(None=ComponentMask(None=TextureCoordinate()) {r=true})
  n31 = TextureSampleParameter2D(UVs=TextureCoordinate() {uTiling=0.5}) {group=Dissolve, parameterName=Dissolve Tex, texture=/Game/Global/Textures/Default/Noise/T_FireNoise_01}
  n32 = Power(
   Base=LinearInterpolate(A=@n30, B=Multiply(A=@n30, B=@n31.R), Alpha=ScalarParameter() {defaultValue=0.2, group=Dissolve, parameterName=Lerp Dissolve Tex})
   Exp=ScalarParameter() {defaultValue=1, group=Dissolve, parameterName=Dissolve Power}
  )
  n33 = ScalarParameter() {defaultValue=10, group=Dissolve, parameterName=Dissolve Edge Hardness, sliderMin=0.5}
  n34 = DynamicParameter() {defaultValue=<rgba 0 1 1 1>, paramNames [Random Offset, Spline Width Mult, Dissolve, Dissolve Edge Intensity]}
  n35 = Saturate(None=Subtract(A=Multiply(A=@n32, B=@n33), B=LinearInterpolate(A=@n33, Alpha=@n34.Dissolve) {constB=-1}))
  n36 = Multiply(A=Multiply(A=Multiply(A=@n27, B=@n28.Dissolve Edge Intensity), B=ComponentMask(None=@n29) {r=true}), B=Subtract(A=Constant() {r=1}, B=@n35))
  n37 = Add(A=Multiply(A=Multiply(A=@n18.RGB, B=@n19), B=@n5), B=Add(A=Multiply(A=@n26, B=@n5), B=Multiply(A=@n5, B=@n36)))
  n38 = ParticleColor()
  MaterialFunctionCall(Emissive=Multiply(A=Add(A=@n6, B=@n37), B=@n38.RGB)) {materialFunction=&MF_FX_EmissiveAdjustments}
  n39 = Constant() {r=1}
  n40 = VectorParameter() {group=Opacity, parameterName=SmokeTrail UV, sortPriority=1, defaultValue=<rgba 1 1 0 0>}
  n41 = AppendVector(A=@n40.R, B=@n40.R)
  n42 = AppendVector(
   A=ScalarParameter() {defaultValue=1, group=Noise, parameterName=Global Scale U}
   B=ScalarParameter() {defaultValue=1, group=Noise, parameterName=Global Scale V}
  )
  n43 = Panner(Coordinate=Multiply(A=TextureCoordinate(), B=Multiply(A=@n41, B=@n42)), Speed=AppendVector(A=@n40.B, B=@n40.B))
  n44 = ComponentMask(None=TextureCoordinate()) {r=true}
  n45 = Multiply(A=Power(Base=OneMinus(None=@n44)) {constExponent=1}) {constB=20}
  n46 = Multiply(A=ComponentMask(None=TextureCoordinate()) {r=true}) {constB=10}
  n47 = Multiply(
   A=TextureSampleParameter2D(UVs=Add(A=@n12, B=@n43)) {group=Opacity, parameterName=OpacityMask, texture=/Game/Effects/Fort_Effects/Textures/Smoke/T_Smoke_Trail_Soft_Chromatic}.R
   B=Multiply(A=Saturate(None=@n45), B=Saturate(None=@n46))
  )
  n48 = Multiply(
   A=Multiply(
    A=ScalarParameter() {defaultValue=1, group=Opacity, parameterName=Opactiy Mult}
    B=StaticSwitchParameter(True=@n47, False=@n39) {group=Opacity, parameterName=Use SmokeTrail Mask}
   )
   B=@n35
  )
  n49 = ScalarParameter() {defaultValue=10, group=Opacity, parameterName=DepthFadeDistance}
  Multiply(A=@n38.A, B=ShadingPathSwitch(Default=DepthFade(Opacity=@n48, FadeDistance=@n49), Mobile=@n48) {desc=Migrated from FeatureLevelSwitch})
  ScalarParameter() {defaultValue=0.5, group=Dissolve, parameterName=Dissolve Dynamic Test}
  n50 = VectorParameter() {group=WPO, parameterName=WPO UV, sortPriority=6, defaultValue=<rgba 1 0.2 -0.6 0>}
  n51 = Multiply(
   A=AppendVector(A=@n50.R, B=@n50.R)
   B=AppendVector(
    A=ScalarParameter() {defaultValue=1, group=Noise, parameterName=Global Scale U}
    B=ScalarParameter() {defaultValue=1, group=Noise, parameterName=Global Scale V}
   )
  )
  n52 = TextureCoordinate()
  n53 = TextureSample(UVs=Panner(Coordinate=Multiply(A=@n52, B=@n51), Speed=AppendVector(A=@n50.B, B=@n50.B))) {samplerType=SAMPLERTYPE_Normal, texture=&T_Weap_Distort_N_02}
  n54 = Constant()
  Transform(None=Multiply(
    A=MaterialFunctionCall(R=@n53.R, G=@n53.R, B=@n54) {materialFunction=&AppendMany}.RGB
    B=ScalarParameter() {defaultValue=5, group=WPO, parameterName=WPO Intensity}
   ) {constB=10}) {transformSourceType=TRANSFORMSOURCE_Local}
 ]
}
