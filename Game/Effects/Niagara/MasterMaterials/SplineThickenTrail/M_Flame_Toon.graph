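# Translucent, unlit, two-sided master material for stylized toon flame trails on Niagara ribbons and meshes.
# World position offset combines SplineThicken width with a time-driven sine wave, both varying along the UVs.
# Emissive adds a panned flame texture to a noise-modulated color gradient, tinted by particle color.
# Opacity layers flame masks, caustics erosion, start/end fades, depth fade and particle alpha, with dynamic parameters driving UV offset and HeightLerp dissolve.
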Material /Game/Effects/Niagara/MasterMaterials/SplineThickenTrail/M_Flame_Toon
graph {
 properties {
  bUsedWithMeshDeformer=true
  bUsedWithNiagaraMeshParticles=true
  bUsedWithNiagaraRibbons=true
  bUsedWithSkeletalMesh=true
  blendMode=BLEND_Translucent
  pixelDepthOffsetMode=PDOM_Legacy
  shadingModel=MSM_Unlit
  twoSided=true
  usedShadingModels=Unlit
 }
 nodes [
  n1 = ScalarParameter() {defaultValue=0.4, parameterName=Width}
  n2 = Multiply(A=Constant() {r=10}, B=@n1)
  n3 = ComponentMask(None=NamedRerouteUsage()) {r=true}
  n4 = ScalarParameter() {defaultValue=5, parameterName=SplineTicken Edge Contrast}
  n5 = LinearInterpolate(Alpha=Saturate(None=Multiply(A=@n3, B=@n4) {constB=5})) {constA=0.3}
  n6 = Multiply(A=Constant() {r=40}, B=@n1)
  n7 = DynamicParameter() {defaultValue=<rgba 1 0 -1 1>, paramNames [Use dissolve, U_Offset, ErodeSpeed, Param4]}
  n8 = Constant()
  n9 = ComponentMask(None=Add(A=NamedRerouteUsage(), B=AppendVector(A=@n7.U_Offset, B=@n8))) {r=true}
  n10 = Multiply(A=Time()) {constB=-1.3}
  n11 = Multiply(
   A=Sine(None=Add(A=@n9, B=@n10)) {period=1.2}
   B=Transform(None=VectorParameter() {parameterName=WPO Direction, defaultValue=<rgba 1 0 0 1>}.RGB) {transformSourceType=TRANSFORMSOURCE_Local}
  )
  n12 = NamedRerouteUsage()
  n13 = Saturate(None=Multiply(A=ComponentMask(None=@n12) {r=true}) {constB=3})
  Add(
   A=MaterialFunctionCall(WidthBase=Multiply(A=@n2, B=@n5), WidthTip=Multiply(A=@n6, B=@n5), Expand U or V UV Channel=StaticBoolParameter() {parameterName=ExpandUorV}) {materialFunction=&SplineThicken}.WorldPositionOffset
   B=Multiply(A=Multiply(A=@n11) {constB=1.7}, B=LinearInterpolate(Alpha=@n13))
  ) {constB=0}
  n14 = Multiply(A=NamedRerouteUsage(), B=ComponentMask(None=VectorParameter() {parameterName=UV Scale, defaultValue=<rgba 0.3 1 0 0>}.RGB) {g=true, r=true})
  n15 = Power(Base=ComponentMask(None=@n14) {r=true}) {constExponent=0.7}
  n16 = ComponentMask(None=@n14) {g=true}
  n17 = AppendVector(A=DynamicParameter() {defaultValue=<rgba 1 0 -1 1>, paramNames [Use dissolve, U_Offset, ErodeSpeed, Param4]}.U_Offset, B=Constant())
  n18 = Panner(
   Coordinate=Add(A=AppendVector(A=@n15, B=@n16), B=@n17) {constB=0}
   Speed=ComponentMask(None=VectorParameter() {parameterName=TrailScrollSpeed, defaultValue=<rgba -0.3 0 0 1>}.RGB) {g=true, r=true}
  ) {speedX=-0.3}
  n19 = Multiply(
   A=VectorParameter() {parameterName=CenterColor, defaultValue=<rgba 0.3 0.0611 0.0188 1>}.RGB
   B=TextureSample(UVs=@n18) {texture=/Game/Effects/Niagara/MasterMaterials/SplineThickenTrail/Texture/T_FlameToon_Trail}.B
  )
  n20 = NamedRerouteUsage()
  n21 = AppendVector(A=DynamicParameter() {defaultValue=<rgba 1 0 -1 1>, paramNames [Use dissolve, U_Offset, ErodeSpeed, Param4]}.U_Offset, B=Constant())
  n22 = TextureSample(UVs=Panner(Coordinate=Add(A=@n20, B=@n21)) {speedX=-1}) {samplerType=SAMPLERTYPE_LinearGrayscale, texture=/Engine/Functions/Engine_MaterialFunctions02/ExampleContent/Textures/LowResBlurredNoise}
  n23 = Clamp(None=LinearInterpolate(Alpha=@n22.R) {constA=-0.8}, Min=ScalarParameter() {defaultValue=1, parameterName=Dark noise min}) {minDefault=0.1}
  n24 = LinearInterpolate(
   A=VectorParameter() {parameterName=ColorLerpA, defaultValue=<rgba 7 0.2188 0.2188 1>}.RGB
   B=VectorParameter() {parameterName=ColorLerpB, defaultValue=<rgba 1 0.0701 0.101 1>}.RGB
   Alpha=ComponentMask(None=NamedRerouteUsage()) {r=true}
  )
  n25 = Multiply(
   A=LinearInterpolate(B=Multiply(A=@n23, B=@n24), Alpha=ScalarParameter() {defaultValue=1, parameterName=Use Color Curve}) {constA=1}
   B=VectorParameter() {parameterName=Main Color, defaultValue=<rgba 0.3542 0 0.2947 1>}.RGB
  )
  MaterialFunctionCall(Emissive=Multiply(A=Add(A=@n19, B=@n25), B=ParticleColor().RGB)) {materialFunction=&MF_FX_EmissiveAdjustments}
  n26 = TextureCoordinate() {uTiling=3, vTiling=2}
  n27 = TextureSample(UVs=Panner(Coordinate=Add(A=@n26, B=@n17) {constB=0}) {speedX=-0.5}) {samplerType=SAMPLERTYPE_Grayscale, texture=/Game/Global/Textures/Default/Noise/LowResBlurredNoise}
  n28 = Multiply(A=NamedRerouteUsage(), B=ComponentMask(None=VectorParameter() {parameterName=UV Scale, defaultValue=<rgba 0.3 1 0 0>}.RGB) {g=true, r=true})
  n29 = Power(Base=ComponentMask(None=@n28) {r=true}) {constExponent=0.7}
  n30 = ComponentMask(None=@n28) {g=true}
  n31 = DynamicParameter() {defaultValue=<rgba 1 0 -1 1>, paramNames [Use dissolve, U_Offset, ErodeSpeed, Param4]}
  n32 = Constant()
  n33 = Panner(Coordinate=Add(A=AppendVector(A=@n29, B=@n30), B=AppendVector(A=@n31.U_Offset, B=@n32)) {constB=0}) {speedX=-0.3}
  n34 = Multiply(
   A=StaticComponentMaskParameter(None=TextureSampleParameter2D(UVs=@n33) {parameterName=Tex_Default, texture=&T_FlameToon_Trail}.RGB) {defaultB=true, parameterName=TexRGBMask}
   B=Saturate(None=LinearInterpolate(Alpha=@n27.R) {constB=1.5})
  )
  n35 = ScalarParameter() {defaultValue=0.5, parameterName=Lerp Center Dissolve}
  n36 = ChannelMaskParameter(None=TextureSampleParameter2D(UVs=@n18) {parameterName=Tex_Default, texture=&T_FlameToon_Trail}.RGBA) {group=Tex, parameterName=Mask}
  n37 = ScalarParameter() {defaultValue=1, group=Tex, parameterName=Tex alpha fix}
  n38 = NamedRerouteUsage()
  n39 = Subtract(A=OneMinus(None=ComponentMask(None=@n38) {r=true}), B=ScalarParameter() {parameterName=EndFade}) {constB=0}
  n40 = Multiply(
   A=Subtract(A=Power(Base=@n36.R, Exp=@n37) {constExponent=2.2}, B=LinearInterpolate(B=@n34, Alpha=@n35))
   B=Saturate(None=Multiply(A=@n39) {constB=5})
  )
  n41 = NamedRerouteUsage()
  n42 = AppendVector(A=DynamicParameter() {defaultValue=<rgba 1 0 -1 1>, paramNames [Use dissolve, U_Offset, ErodeSpeed, Param4]}.U_Offset, B=Constant())
  n43 = VectorParameter() {parameterName=DissolveScrollSpeed, defaultValue=<rgba -0.5 0 0 1>}
  n44 = TextureSample(UVs=Panner(Coordinate=Add(A=@n41, B=@n42) {constB=0}, Speed=ComponentMask(None=@n43.RGB) {g=true, r=true}) {speedX=-0.5}) {samplerType=SAMPLERTYPE_LinearGrayscale, texture=/Game/Global/Textures/Default/Caustics/T_Nosie_Caustics_Grayscale}
  n45 = ComponentMask(None=TextureCoordinate()) {r=true}
  n46 = Saturate(None=LinearInterpolate(
    A=ScalarParameter() {parameterName=Dissolve mask Min}
    B=ScalarParameter() {defaultValue=0.5, parameterName=Dissolve mask Max}
    Alpha=Power(Base=@n45) {constExponent=1.3}
   ) {constA=0.1, constB=2})
  n47 = Saturate(None=LinearInterpolate(
    A=ScalarParameter() {defaultValue=-2, parameterName=Disappear min}
    B=ScalarParameter() {defaultValue=2, parameterName=Disappear max}
    Alpha=Power(Base=@n44.R, Exp=@n46)
   ) {constA=-2, constB=2})
  n48 = Subtract(A=ComponentMask(None=NamedRerouteUsage()) {r=true}, B=ScalarParameter() {parameterName=StartFade})
  n49 = ScalarParameter() {defaultValue=10, parameterName=StartFadeSharpness}
  n50 = Multiply(A=Multiply(A=Multiply(A=@n40, B=@n47)), B=Saturate(None=Multiply(A=@n48, B=@n49) {constB=10}))
  n51 = ScalarParameter() {defaultValue=10, parameterName=FadeDistance}
  n52 = Panner(Coordinate=NamedRerouteUsage()) {speedX=-0.5}
  n53 = Multiply(A=MaterialFunctionCall(
    Transition Phase=DynamicParameter() {defaultValue=<rgba 1 0 -1 1>, paramNames [Use dissolve, U_Offset, ErodeSpeed, Param4]}.Use dissolve
    Height Texture=TextureSample(UVs=@n52) {samplerType=SAMPLERTYPE_LinearGrayscale, texture=&T_Nosie_Caustics_Grayscale}.R
   ) {materialFunction=&HeightLerp}.Alpha)
  Multiply(A=Multiply(A=DepthFade(Opacity=@n50, FadeDistance=@n51), B=LinearInterpolate(Alpha=@n53) {constA=-2, constB=2}), B=ParticleColor().A)
  n54 = TextureCoordinate()
  NamedRerouteDeclaration(None=LinearInterpolate(
    A=@n54
    B=MaterialFunctionCall(XY=@n54) {materialFunction=&Swizzle}.YX
    Alpha=ScalarParameter() {parameterName=SwizzleUV}
   )) {name=UV, nodeColor=<rgba 0.0312 1 0 1>}
  StaticBool()
  TextureCoordinate()
  TextureCoordinate()
  TextureCoordinate() {uTiling=2, vTiling=0.8}
  TextureCoordinate()
  TextureCoordinate() {uTiling=2.1, vTiling=0.85}
  TextureCoordinate()
  TextureCoordinate() {uTiling=2}
  TextureCoordinate()
  TextureCoordinate()
  TextureCoordinate()
  TextureCoordinate() {uTiling=2.2}
 ]
}
