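# Translucent, unlit, two-sided material for snow trail effects on mesh particles, using SplineThicken with width driven by a dynamic parameter.
# Fakes lighting by dotting a panning normal map with the sun vector to blend two time-of-day palettes, tinted by particle color.
# Opacity combines a sine fade along U, a panning smoke mask, an eroding dissolve texture and particle alpha.
# Dynamic parameters also control smoke UV offset, V scale and dissolve power.
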Material /Game/Effects/Niagara/MasterMaterials/SplineThickenTrail/M_Snow_SplineTicken
graph {
 properties {
  bUsedWithMeshParticles=true
  bUsedWithNiagaraMeshParticles=true
  blendMode=BLEND_Translucent
  pixelDepthOffsetMode=PDOM_Legacy
  shadingModel=MSM_Unlit
  twoSided=true
  usedShadingModels=Unlit
 }
 nodes [
  n1 = DynamicParameter() {defaultValue=<rgba 1 1 1.5 1>, paramNames [Spline_Width, Offset, Dissappear_power, V_Scale]}
  MaterialFunctionCall(
   WidthBase=Multiply(A=Constant() {r=30}, B=@n1.Spline_Width)
   WidthTip=Multiply(A=Constant() {r=50}, B=@n1.Spline_Width)
   Expand U or V UV Channel=StaticBool()
  ) {materialFunction=&SplineThicken}
  n2 = Constant() {r=1}
  n3 = DynamicParameter() {defaultValue=<rgba 1 1 1.5 1>, paramNames [Spline_Width, Offset, Dissappear_power, V_Scale]}
  n4 = Add(
   A=Multiply(A=TextureCoordinate() {uTiling=0.4}, B=AppendVector(A=@n2, B=@n3.V_Scale))
   B=AppendVector(A=Constant(), B=DynamicParameter() {defaultValue=<rgba 1 1 1.5 1>, paramNames [Spline_Width, Offset, Dissappear_power, V_Scale]}.Offset)
  )
  MaterialFunctionCall(XY=@n4) {materialFunction=&Swizzle}
  MaterialFunctionCall(XY=@n4) {materialFunction=&Swizzle}
  n5 = ParticleColor()
  n6 = MaterialFunctionCall(
   Color_Morning=VectorParameter() {parameterName=Subsurface Color morning 1, defaultValue=<rgba 0.1867 0.2018 0.2344 0>}.RGB
   Color_Day=VectorParameter() {parameterName=Subsurface Color day 1, defaultValue=<rgba 0.1796 0.2791 0.4479 1>}.RGB
   Color_Evening=VectorParameter() {parameterName=Subsurface Color evening 1, defaultValue=<rgba 0.2664 0.2122 0.2384 0>}.RGB
   Color_Night=VectorParameter() {parameterName=Subsurface Color night 1, defaultValue=<rgba 0.2915 0.3721 0.5089 1>}.RGB
  ) {materialFunction=&Color_Time_of_Day}
  n7 = MaterialFunctionCall(
   Color_Morning=VectorParameter() {parameterName=Subsurface Color morning 2, defaultValue=<rgba 0.6562 0.188 0.4299 0>}.RGB
   Color_Day=VectorParameter() {parameterName=Subsurface Color day 2, defaultValue=<rgba 0.7083 0.8128 1 1>}.RGB
   Color_Evening=VectorParameter() {parameterName=Subsurface Color evening 2, defaultValue=<rgba 0.3916 0.5776 0.7011 0>}.RGB
   Color_Night=VectorParameter() {parameterName=Subsurface Color night 2, defaultValue=<rgba 0.1639 0.2348 0.3542 1>}.RGB
  ) {materialFunction=&Color_Time_of_Day}
  n8 = CollectionParameter() {
   collection=&FortniteMaterialParameters
   parameterName=SunlightDirectionalVector
  }
  n9 = ComponentMask(None=@n8) {b=true, g=true, r=true}
  n10 = TextureSample(UVs=Panner(Coordinate=@n4) {speedX=-0.5}) {samplerType=SAMPLERTYPE_Normal, texture=/Game/Effects/Fort_Effects/Gameplay/SmokeDustExplosions/Textures/T_RibbonSmokeMasked_01}
  n11 = Power(Base=Saturate(None=DotProduct(A=@n9, B=@n10.RGB)), Exp=ScalarParameter() {defaultValue=3, parameterName=Power}) {constExponent=3}
  MaterialFunctionCall(Emissive=Multiply(A=@n5.RGB, B=LinearInterpolate(A=@n6.Result, B=@n7.Result, Alpha=@n11))) {materialFunction=&MF_FX_EmissiveAdjustments}
  n12 = TextureCoordinate()
  n13 = Multiply(
   A=Sine(None=ComponentMask(None=@n12) {r=true}) {period=2}
   B=TextureSample(UVs=Panner(Coordinate=@n4) {speedX=-0.5}) {texture=/Game/Effects/Fort_Effects/Gameplay/SmokeDustExplosions/Textures/T_RibbonSmokeMasked_02}.R
  )
  n14 = Multiply(A=TextureCoordinate() {uTiling=0.5, vTiling=0.5})
  n15 = Power(
   Base=TextureSample(UVs=Panner(Coordinate=@n14) {speedX=-1}) {texture=/Game/CharClasses/Outlander/Gadgets/AntiMaterialCharge/Materials/T_AntiMaterialCharge_Dissolve_01}.RGB
   Exp=DynamicParameter() {defaultValue=<rgba 1 1 1.5 1>, paramNames [Spline_Width, Offset, Dissappear_power, V_Scale]}.Dissappear_power
  ) {constExponent=1.5}
  n16 = ComponentMask(None=TextureCoordinate()) {r=true}
  n17 = Multiply(A=Power(Base=Add(A=@n16) {constB=-0.3})) {constB=10}
  n18 = Saturate(None=Multiply(A=Power(Base=@n15, Exp=@n17)) {constB=5})
  Multiply(A=Multiply(A=@n5.A, B=Multiply(A=@n13, B=@n18)), B=ScalarParameter() {defaultValue=1, parameterName=Alpha})
  OneMinus(None=@n16)
 ]
}
