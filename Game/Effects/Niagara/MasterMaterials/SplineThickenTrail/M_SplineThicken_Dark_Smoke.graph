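# Translucent, unlit, two-sided master material for dark smoke trails on Niagara ribbons and mesh particles.
# Opacity combines panning, noise-distorted smoke, trail and caustic textures with UV tail and appear masks driven by dynamic parameters, then depth, view-angle and camera fades.
# Emissive uses particle color, optionally tinted by a rotating camera-vector triplanar galaxy texture.
# WPO adds a camera-space sine wobble to optional SplineThicken, whose tip width also reads WidthBase, leaving WidthTip unused.
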Material /Game/Effects/Niagara/MasterMaterials/SplineThickenTrail/M_SplineThicken_Dark_Smoke
graph {
 properties {
  bUsedWithNiagaraMeshParticles=true
  bUsedWithNiagaraRibbons=true
  blendMode=BLEND_Translucent
  pixelDepthOffsetMode=PDOM_Legacy
  shadingModel=MSM_Unlit
  twoSided=true
  usedShadingModels=Unlit
 }
 nodes [
  n1 = ParticleColor()
  n2 = Time()
  n3 = LinearInterpolate(A=@n2, B=Multiply(A=@n2, B=Constant()), Alpha=ScalarParameter() {group=Galaxy, parameterName=Background Rotation})
  n4 = AppendVector(
   A=AppendVector(
    A=ScalarParameter() {group=Galaxy, parameterName=Galaxy Rotation Speed}
    B=ScalarParameter() {group=Galaxy, parameterName=Galaxy Rotation Speed_1}
   )
   B=ScalarParameter() {group=Galaxy, parameterName=Galaxy Rotation Speed_2}
  )
  n5 = TextureObjectParameter() {
   group=Galaxy
   parameterName=GalaxyTexture
   texture=/Game/Characters/Player/Male/Medium/Bodies/M_Med_Soldier_04/Skins/Galaxy/Textures/T_FN_Nebula_style_Green
  }
  n6 = Desaturation(
   None=MaterialFunctionCall!(
    CameraVector
    Rotation Speed=Multiply(A=@n3, B=@n4)
    Texture=@n5
    Tiling=ScalarParameter() {defaultValue=1, group=Galaxy, parameterName=GalaxyTiling}
    Offset=Constant()
    AxisFadeContrast=ScalarParameter() {defaultValue=6, group=Galaxy, parameterName=GalaxyAxisFade}
   ) {materialFunction=/Game/Weapons/WeaponSkins/MaterialFunction/TriplanarCameraVector_Rotated}.Result
   Fraction=ScalarParameter() {group=Galaxy, parameterName=Galaxy Saturation}
  )
  MaterialFunctionCall(Emissive=LinearInterpolate(A=@n1.RGB, B=Multiply(A=@n6, B=@n1.RGB), Alpha=ScalarParameter() {parameterName=Use_EmissiveTexture})) {materialFunction=&MF_FX_EmissiveAdjustments}
  n7 = DynamicParameter() {parameterIndex=1, defaultValue=<rgba 0 1 0 0>, paramNames [Use Forward Tail mask, Appear Mask, TailMinimum, none]}
  n8 = Multiply(
   A=TextureCoordinate()
   B=AppendVector(
    A=ScalarParameter() {defaultValue=1, parameterName=Forward Tail mask U}
    B=ScalarParameter() {defaultValue=0.6, parameterName=Forward Tail mask V}
   )
  )
  n9 = AppendVector(A=DynamicParameter() {defaultValue=<rgba 200 50 1 0>, paramNames [WidthBase, WidthTip, U_Scale, U offset]}.U offset, B=Constant())
  n10 = Multiply(A=ScalarParameter() {defaultValue=-1, parameterName=PannerSpeed}) {constB=0.5}
  n11 = Constant()
  n12 = TextureSample(UVs=Panner(Coordinate=Add(A=@n8, B=@n9), Speed=AppendVector(A=@n10, B=@n11)) {speedX=-0.5}) {samplerType=SAMPLERTYPE_Masks, texture=/Game/Weapons/WeaponSkins/Wraps/Textures/Mash/Mash_Noise}
  n13 = DynamicParameter() {parameterIndex=1, defaultValue=<rgba 0 1 0 0>, paramNames [Use Forward Tail mask, Appear Mask, TailMinimum, none]}
  n14 = ComponentMask(None=TextureCoordinate()) {r=true}
  n15 = Saturate(None=LinearInterpolate(
    A=Constant() {r=-1}
    B=LinearInterpolate(Alpha=@n13.Use Forward Tail mask) {constA=0.5, constB=1.5}
    Alpha=Power(Base=@n14) {constExponent=1.3}
   ) {constA=0.1, constB=2})
  n16 = Power(Base=@n12.R, Exp=@n15)
  n17 = Power(Base=OneMinus(None=@n15), Exp=ScalarParameter() {defaultValue=5, parameterName=Forward Tail Mask Power}) {constExponent=5}
  n18 = Saturate(None=LinearInterpolate(B=@n12.R, Alpha=OneMinus(None=@n17)) {constA=1})
  n19 = ScalarParameter() {parameterName=Use alternate masking}
  n20 = Saturate(None=LinearInterpolate(
    A=LinearInterpolate(Alpha=@n7.Use Forward Tail mask) {constB=-0.2}
    B=Constant() {r=1.2}
    Alpha=LinearInterpolate(A=@n16, B=@n18, Alpha=@n19) {constAlpha=1}
   ) {constA=-2, constB=2})
  n21 = ScalarParameter() {defaultValue=-1, parameterName=PannerSpeed}
  n22 = Constant()
  n23 = TextureSample(UVs=Panner(Coordinate=TextureCoordinate(), Speed=AppendVector(A=@n21, B=@n22)) {speedX=-1}) {samplerType=SAMPLERTYPE_Grayscale, texture=&T_Smoke_Trail_Soft}
  n24 = TextureCoordinate() {uTiling=2}
  n25 = Panner(
   Coordinate=TextureCoordinate() {uTiling=0.8, vTiling=0.5}
   Speed=Multiply(A=Constant2Vector() {g=0.2, r=-0.1}, B=ScalarParameter() {defaultValue=1, parameterName=PannerSpeed 2})
  ) {speedX=-0.1, speedY=0.2}
  n26 = Multiply(A=ComponentMask(None=TextureSample(UVs=@n25) {samplerType=SAMPLERTYPE_Masks, texture=&Mash_Noise}.RGB) {g=true, r=true}) {constB=0.12}
  n27 = Constant2Vector() {r=-1}
  n28 = ScalarParameter() {defaultValue=1, parameterName=PannerSpeed 2}
  n29 = TextureSample(UVs=Panner(Coordinate=Add(A=@n24, B=@n26), Speed=Multiply(A=@n27, B=@n28)) {speedX=-1}) {texture=/Game/Gadgets/Assets/VinderTech_GliderChute/Glider_DarkStance/FX/T_Glider_DarkStance_Trail}
  n30 = Power(Base=Add(A=Multiply(A=@n23.R), B=Multiply(A=@n29.G) {constB=2}) {constB=0})
  n31 = DynamicParameter() {defaultValue=<rgba 200 50 1 0>, paramNames [WidthBase, WidthTip, U_Scale, U offset]}
  n32 = Constant() {r=1}
  n33 = Add(A=Multiply(A=AppendVector(A=@n31.U_Scale, B=@n32), B=TextureCoordinate()), B=@n26)
  n34 = AppendVector(A=ScalarParameter() {defaultValue=-1, parameterName=PannerSpeed}, B=Constant())
  n35 = Multiply(A=ScalarParameter() {defaultValue=-1, parameterName=PannerSpeed}) {constB=0.5}
  n36 = Constant()
  n37 = TextureSample(UVs=Panner(Coordinate=TextureCoordinate(), Speed=AppendVector(A=@n35, B=@n36)) {speedX=-0.5}) {
   samplerType=SAMPLERTYPE_LinearColor
   texture=/Engine/Functions/Engine_MaterialFunctions02/ExampleContent/Textures/OffsetNoiseDistanceFields_ForNormals
  }
  n38 = Multiply(
   A=TextureSample(UVs=Panner(Coordinate=@n33, Speed=@n34) {speedX=-1}) {texture=&T_Glider_DarkStance_Trail}.R
   B=LinearInterpolate(Alpha=@n37.R) {constA=0.4}
  )
  n39 = Saturate(None=Add(A=Add(A=@n30, B=@n38)) {constB=0})
  n40 = Power(Base=ComponentMask(None=TextureCoordinate()) {r=true}) {constExponent=1.5}
  n41 = Multiply(A=@n39, B=Power(Base=LinearInterpolate(Alpha=@n37.R) {constA=0.5}, Exp=LinearInterpolate(Alpha=@n40)) {constExponent=1})
  n42 = LinearInterpolate(Alpha=ComponentMask(None=TextureCoordinate()) {r=true}) {constA=0.8}
  n43 = Multiply(
   A=AppendVector(A=DynamicParameter() {defaultValue=<rgba 200 50 1 0>, paramNames [WidthBase, WidthTip, U_Scale, U offset]}.U_Scale, B=Constant() {r=1})
   B=TextureCoordinate() {uTiling=1.1, vTiling=1.7}
  )
  n44 = AppendVector(A=ScalarParameter() {defaultValue=-1, parameterName=PannerSpeed}, B=Constant())
  n45 = TextureSample(UVs=Add(A=Panner(Coordinate=@n43, Speed=@n44) {speedX=-1}, B=Multiply(A=@n26))) {samplerType=SAMPLERTYPE_LinearGrayscale, texture=/Game/Global/Textures/Default/Caustics/T_Nosie_Caustics_Grayscale}
  n46 = ComponentMask(None=TextureCoordinate()) {r=true}
  n47 = Multiply(A=Multiply(A=Power(Base=@n45.R) {constExponent=1}) {constB=1.5}, B=Power(Base=OneMinus(None=@n46)) {constExponent=1})
  n48 = Saturate(None=Add(A=Multiply(A=@n41, B=@n42), B=Multiply(A=@n47, B=@n39)))
  n49 = TextureCoordinate()
  n50 = Multiply(A=OneMinus(None=ComponentMask(None=@n49) {r=true})) {constB=10}
  n51 = Subtract(A=ComponentMask(None=TextureCoordinate()) {r=true}, B=ScalarParameter() {defaultValue=0.1, parameterName=LeftMask_Subtract}) {constB=0.1}
  n52 = Multiply(A=Multiply(A=Multiply(A=@n48, B=@n20), B=Saturate(None=@n50)), B=Saturate(None=Multiply(A=@n51) {constB=3}))
  n53 = ComponentMask(None=TextureCoordinate()) {r=true}
  n54 = DynamicParameter() {parameterIndex=1, defaultValue=<rgba 0 1 0 0>, paramNames [Use Forward Tail mask, Appear Mask, TailMinimum, none]}
  n55 = Multiply(A=Add(A=OneMinus(None=@n53), B=LinearInterpolate(Alpha=@n54.Appear Mask) {constA=-1, constB=0.1})) {constB=10}
  n56 = TextureCoordinate()
  n57 = Multiply(A=OneMinus(None=ComponentMask(None=@n56) {r=true})) {constB=2}
  n58 = DynamicParameter() {parameterIndex=1, defaultValue=<rgba 0 1 0 0>, paramNames [Use Forward Tail mask, Appear Mask, TailMinimum, none]}
  n59 = Multiply(A=Power(Base=Saturate(None=@n57), Exp=@n58.TailMinimum) {constExponent=3}) {constB=3}
  n60 = Multiply(
   A=Multiply(A=Multiply(A=@n1.A, B=@n52), B=Saturate(None=@n55))
   B=LinearInterpolate(B=Saturate(None=@n59), Alpha=ScalarParameter() {parameterName=Use Custom TailMinimum}) {constA=1}
  )
  n61 = DepthFade(Opacity=@n60, FadeDistance=ScalarParameter() {parameterName=DepthFade})
  n62 = Multiply(A=MaterialFunctionCall("FadeSensitivity (optional)"=ScalarParameter() {parameterName="FadeSensitivity (optional) (S)"}) {materialFunction=&FadeBasedOnViewAngle}.Result) {constB=10}
  Multiply(
   A=Multiply(A=ShadingPathSwitch(Default=@n61, Deferred=@n61, Forward=@n60, Mobile=@n60), B=Saturate(None=@n62))
   B=MaterialFunctionCall(Fade Length=ScalarParameter() {parameterName="cam Fade Length (S)"}, Fade Offset=ScalarParameter() {parameterName="cam Fade Offset (S)"}) {materialFunction=&CameraDepthFade}.Result
  )
  n63 = DynamicParameter() {defaultValue=<rgba 200 50 1 0>, paramNames [WidthBase, WidthTip, U_Scale, U offset]}
  n64 = LinearInterpolate(
   A=Constant()
   B=MaterialFunctionCall(
    WidthBase=@n63.WidthBase
    WidthTip=@n63.WidthBase
    Expand U or V UV Channel=StaticBoolParameter() {parameterName="Expand U or V UV Channel (SB)"}
   ) {materialFunction=&SplineThicken}.WorldPositionOffset
   Alpha=ScalarParameter() {defaultValue=1, parameterName=Use_SplineThicken}
  )
  n65 = TextureCoordinate()
  n66 = DynamicParameter() {defaultValue=<rgba 200 50 1 0>, paramNames [WidthBase, WidthTip, U_Scale, U offset]}
  n67 = Add(A=@n65, B=AppendVector(A=@n66.U offset, B=Constant()))
  n68 = ScalarParameter() {defaultValue=1, parameterName=Use U Offset on WPO}
  n69 = Constant() {r=1}
  n70 = ComponentMask(None=Multiply(A=LinearInterpolate(A=@n65, B=@n67, Alpha=@n68), B=AppendVector(A=@n66.U_Scale, B=@n69))) {r=true}
  n71 = Multiply(A=Time(), B=ScalarParameter() {defaultValue=-0.7, parameterName=WPO Time Mult}) {constB=-0.7}
  n72 = Sine(None=Multiply(A=Add(A=@n70, B=@n71), B=ScalarParameter() {defaultValue=1, parameterName=WPO Frequency}))
  n73 = Transform(None=VectorParameter() {parameterName=WPO Direction, defaultValue=<rgba 0 1 0 1>}.RGB) {transformSourceType=TRANSFORMSOURCE_Camera}
  n74 = ComponentMask(None=TextureCoordinate()) {r=true}
  n75 = Multiply(
   A=Multiply(A=Multiply(A=@n72, B=@n73), B=ScalarParameter() {defaultValue=5, parameterName=WPO Intensity})
   B=Saturate(None=Multiply(A=@n74) {constB=7})
  )
  Reroute(None=Add(A=Add(A=@n64, B=@n75)) {constB=0})
 ]
}
