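# Translucent unlit Niagara master material for flame trails on ribbons and mesh particles.
# World position offset combines SplineThicken width, optional inertia bend and sine wobble, driven by dynamic parameters.
# Opacity multiplies panned noise and flame textures by erosion, tail and lead masks, with optional angle fade, height-lerp dissolve and appear wipe, then depth fade.
# Emissive blends two colors along U, tinted by particle color, through MF_FX_EmissiveAdjustments.
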Material /Game/Effects/Niagara/MasterMaterials/SplineThickenTrail/M_SplineThicken_Flame
graph {
 properties {
  bUsedWithNiagaraMeshParticles=true
  bUsedWithNiagaraRibbons=true
  blendMode=BLEND_Translucent
  pixelDepthOffsetMode=PDOM_Legacy
  shadingModel=MSM_Unlit
  usedShadingModels=Unlit
 }
 nodes [
  n1 = DynamicParameter() {defaultValue=<rgba 5 15 0 1>, paramNames [WidthBase,WidthTip,UOffset,SpeedOffset]}
  n2 = MaterialFunctionCall(
   WidthBase=@n1.WidthBase
   WidthTip=@n1.WidthBase
   Expand U or V UV Channel=StaticBoolParameter() {parameterName="Expand U or V UV Channel (SB)"}
  ) {materialFunction=&SplineThicken}
  n3 = ComponentMask(None=TextureCoordinate()) {r=true}
  n4 = ScalarParameter() {defaultValue=3, group=Inertia, parameterName=Inertia Gradient Power, sortPriority=2}
  n5 = LinearInterpolate(
   B=LinearInterpolate(
    B=DynamicParameter() {parameterIndex=1, defaultValue=<rgba 0 0 0 1>, paramNames [X,Y,Z,Appear]}.RGB
    Alpha=Power(Base=@n3, Exp=@n4) {constExponent=3}
   ) {constAlpha=0, constB=0}
   Alpha=ScalarParameter() {group=Inertia, parameterName=Use Dynamic Inertia, sliderMax=1, sortPriority=0}
  )
  n6 = ComponentMask(None=TextureCoordinate()) {r=true}
  n7 = Multiply(A=Time()) {constB=-0.5}
  n8 = Time()
  n9 = Multiply(
   A=ScalarParameter() {defaultValue=1, parameterName=TimeMult}
   B=DynamicParameter() {defaultValue=<rgba 5 15 0 1>, paramNames [WidthBase,WidthTip,UOffset,SpeedOffset]}.SpeedOffset
  )
  n10 = TextureSample(UVs=Panner(Coordinate=TextureCoordinate() {uTiling=3, vTiling=0}, Time=Multiply(A=@n8, B=@n9)) {speedX=-0.3}) {samplerType=SAMPLERTYPE_LinearGrayscale, texture=/Engine/Functions/Engine_MaterialFunctions02/ExampleContent/Textures/LowResBlurredNoise}
  n11 = LinearInterpolate(A=Sine(None=Add(A=@n6, B=@n7)) {period=0.4}, B=@n10.R)
  n12 = Transform(None=Multiply(A=Constant3Vector() {constant=<rgba 0 1 0 0>}, B=LinearInterpolate(Alpha=@n11) {constA=-1}))
  n13 = Multiply(
   A=ScalarParameter() {defaultValue=0.5, parameterName=WPOMult}
   B=LinearInterpolate(
    B=DynamicParameter() {parameterIndex=2, paramNames [WPOScale, none, DissolveRange, none]}.WPOScale
    Alpha=ScalarParameter() {parameterName=UseCustom WPO}
   ) {constA=1}
  )
  n14 = TextureCoordinate()
  n15 = Multiply(A=Time(), B=ScalarParameter() {defaultValue=1, parameterName=TimeMult})
  n16 = Sine(None=Add(A=ComponentMask(None=@n14) {r=true}, B=Multiply(A=@n15) {constB=-0.5})) {period=0.4}
  n17 = Transform(None=Multiply(A=VectorParameter() {parameterName=TrailWPODirection, defaultValue=<rgba 0 1 0 1>}.RGB, B=LinearInterpolate(Alpha=@n16))) {transformSourceType=TRANSFORMSOURCE_Local, transformType=TRANSFORM_Instance}
  n18 = DynamicParameter() {parameterIndex=2, paramNames [WPOScale, none, DissolveRange, none]}
  Add(
   A=LinearInterpolate(B=Add(A=@n2.WorldPositionOffset, B=@n5), Alpha=ScalarParameter() {defaultValue=1, parameterName=UseSplineThicken})
   B=LinearInterpolate(A=Multiply(A=@n12, B=@n13) {constB=0.5}, B=Multiply(A=@n17, B=@n18.WPOScale), Alpha=ScalarParameter() {parameterName=UseTrailWpo})
  )
  n19 = ParticleColor()
  n20 = TextureCoordinate() {uTiling=0.8}
  n21 = AppendVector(A=ScalarParameter() {defaultValue=1, parameterName=UScale}, B=ScalarParameter() {defaultValue=1, parameterName=VScale})
  n22 = VectorParameter() {parameterName=UVscale, defaultValue=<rgba 1 1 0 0>}
  n23 = TextureSample(UVs=Multiply(A=Multiply(A=@n20, B=@n21), B=ComponentMask(None=@n22.RGB) {g=true, r=true})) {texture=/Game/Accessories/FORT_Backpacks/Backpack_M_MED_KeyChain/FX/T_Backpack_KeyChain_Flame}
  n24 = VectorParameter() {parameterName=UVscale, defaultValue=<rgba 1 1 0 0>}
  n25 = TextureSampleParameter2D(UVs=Multiply(A=TextureCoordinate(), B=ComponentMask(None=@n24.RGB) {g=true, r=true})) {
   parameterName=TexNoise
   samplerType=SAMPLERTYPE_LinearGrayscale
   texture=&LowResBlurredNoise
  }
  n26 = Multiply(
   A=Add(A=TextureCoordinate() {uTiling=2, vTiling=0.9}, B=Constant2Vector() {g=0.05})
   B=AppendVector(A=ScalarParameter() {defaultValue=1, parameterName=UScale}, B=ScalarParameter() {defaultValue=1, parameterName=VScale})
  )
  n27 = ComponentMask(None=VectorParameter() {parameterName=UVscale, defaultValue=<rgba 1 1 0 0>}.RGB) {g=true, r=true}
  n28 = Time()
  n29 = Multiply(
   A=ScalarParameter() {defaultValue=1, parameterName=TimeMult}
   B=DynamicParameter() {defaultValue=<rgba 5 15 0 1>, paramNames [WidthBase,WidthTip,UOffset,SpeedOffset]}.SpeedOffset
  )
  n30 = TextureSampleParameter2D(UVs=Panner(Coordinate=Multiply(A=@n26, B=@n27), Time=Multiply(A=@n28, B=@n29)) {speedX=-0.5}) {parameterName=TexA, samplerType=SAMPLERTYPE_LinearGrayscale, texture=/Game/Effects/Fort_Effects/Emotes/UltraLight/Texture/T_PlasmaCoils}
  n31 = DynamicParameter() {defaultValue=<rgba 5 15 0 1>, paramNames [WidthBase,WidthTip,UOffset,SpeedOffset]}
  n32 = Constant()
  n33 = Multiply(
   A=Add(A=TextureCoordinate(), B=AppendVector(A=@n31.UOffset, B=@n32))
   B=AppendVector(A=ScalarParameter() {defaultValue=1, parameterName=UScale}, B=ScalarParameter() {defaultValue=1, parameterName=VScale})
  )
  n34 = ComponentMask(None=VectorParameter() {parameterName=UVscale, defaultValue=<rgba 1 1 0 0>}.RGB) {g=true, r=true}
  n35 = Time()
  n36 = Multiply(A=ScalarParameter() {defaultValue=1, parameterName=TimeMult})
  n37 = TextureSampleParameter2D(UVs=Panner(Coordinate=Multiply(A=@n33, B=@n34), Time=Multiply(A=@n35, B=@n36)) {speedX=-0.5}) {parameterName=CoreTexture, texture=&T_Backpack_KeyChain_Flame}
  n38 = Multiply(A=@n30.R, B=@n37.R)
  n39 = LinearInterpolate(
   A=Power(Base=Multiply(A=@n25.R, B=@n38), Exp=ScalarParameter() {defaultValue=1.1, parameterName=Tex power}) {constExponent=1.1}
   B=@n23.G
   Alpha=ScalarParameter() {parameterName=Use simple glow mask}
  )
  n40 = VectorParameter() {parameterName=ErosionTextureSize, defaultValue=<rgba 1 1 0 1>}
  n41 = Add(A=Multiply(A=TextureCoordinate() {uTiling=2}, B=ComponentMask(None=@n40.RGB) {g=true, r=true})) {constB=0}
  n42 = AppendVector(A=ScalarParameter() {defaultValue=1, parameterName=UScale}, B=ScalarParameter() {defaultValue=1, parameterName=VScale})
  n43 = Time()
  n44 = Multiply(
   A=ScalarParameter() {defaultValue=1, parameterName=TimeMult}
   B=DynamicParameter() {defaultValue=<rgba 5 15 0 1>, paramNames [WidthBase,WidthTip,UOffset,SpeedOffset]}.SpeedOffset
  )
  n45 = TextureSampleParameter2D(UVs=Panner(Coordinate=Multiply(A=@n41, B=@n42), Time=Multiply(A=@n43, B=@n44)) {speedX=-0.7}) {
   parameterName=ErosionTexture
   samplerType=SAMPLERTYPE_LinearGrayscale
   texture=/Game/Global/Textures/Default/Caustics/T_Nosie_Caustics_Grayscale
  }
  n46 = ComponentMask(None=TextureCoordinate()) {r=true}
  n47 = Saturate(None=LinearInterpolate(
    A=ScalarParameter() {parameterName=Dissolve mask Min}
    B=ScalarParameter() {defaultValue=0.5, parameterName=Dissolve mask Max}
    Alpha=Power(Base=@n46) {constExponent=1.3}
   ) {constA=0.1, constB=2})
  n48 = Saturate(None=LinearInterpolate(
    A=ScalarParameter() {defaultValue=-2, parameterName=Disappear min}
    B=ScalarParameter() {defaultValue=2, parameterName=Disappear max}
    Alpha=Power(Base=@n45.R, Exp=@n47) {constExponent=0.2}
   ) {constA=-2, constB=2})
  n49 = OneMinus(None=ComponentMask(None=TextureCoordinate()) {r=true})
  n50 = ScalarParameter() {defaultValue=2, parameterName=Tail Power}
  n51 = Saturate(None=Multiply(A=Power(Base=@n49, Exp=@n50) {constExponent=4}, B=ScalarParameter() {defaultValue=8, parameterName=Tail edge hardness}) {constB=5})
  n52 = ComponentMask(None=TextureCoordinate()) {r=true}
  n53 = ScalarParameter() {defaultValue=1.5, parameterName=LeadMaskPower}
  n54 = Saturate(None=Multiply(A=Power(Base=@n52, Exp=@n53) {constExponent=1.5}, B=ScalarParameter() {defaultValue=10, parameterName=LeadMaskHardness}) {constB=10})
  n55 = Multiply(A=Multiply(A=Multiply(A=@n39, B=@n48), B=Multiply(A=@n51, B=@n54))) {constB=2}
  n56 = MaterialFunctionCall("FadeSensitivity (optional)"=ScalarParameter() {parameterName=AngleFadeSensitivity}) {materialFunction=&FadeBasedOnViewAngle}
  n57 = Multiply(A=@n19.A, B=LinearInterpolate(A=@n55, B=Multiply(A=@n55, B=@n56.Result), Alpha=ScalarParameter() {parameterName=UseAngleFade}))
  n58 = DynamicParameter() {parameterIndex=2, paramNames [WPOScale, none, DissolveRange, none]}
  n59 = TextureSample(UVs=Panner(Coordinate=TextureCoordinate() {uTiling=1.5, vTiling=0.5}) {speedX=-1.5}) {texture=/Game/Effects/Niagara/MasterMaterials/SplineThickenTrail/Texture/T_Pickaxe_SoapPocket_NoiseGlow}
  n60 = Constant() {r=1}
  n61 = LinearInterpolate(
   B=Multiply(A=MaterialFunctionCall(Transition Phase=@n58.DissolveRange, Height Texture=@n59.R, Contrast=@n60) {materialFunction=&HeightLerp}.Alpha)
   Alpha=ScalarParameter() {parameterName=UseGlobleDissolve}
  ) {constA=1}
  n62 = OneMinus(None=ComponentMask(None=TextureCoordinate()) {r=true})
  n63 = LinearInterpolate(Alpha=DynamicParameter() {parameterIndex=1, defaultValue=<rgba 0 0 0 1>, paramNames [X,Y,Z,Appear]}.Appear) {constA=-1}
  n64 = Saturate(None=Multiply(A=Add(A=@n62, B=@n63)) {constB=3})
  n65 = ScalarParameter() {parameterName=UseAppear}
  DepthFade(
   Opacity=Multiply(A=Multiply(A=@n57, B=@n61), B=LinearInterpolate(B=@n64, Alpha=@n65) {constA=1})
   FadeDistance=ScalarParameter() {defaultValue=5, parameterName=DepthFadeDistance}
  ) {fadeDistanceDefault=5}
  n66 = VectorParameter() {parameterName=Color1, defaultValue=<rgba 5 0.1667 0 1>}
  n67 = VectorParameter() {parameterName=Color2, defaultValue=<rgba 15 1.25 0 1>}
  n68 = TextureCoordinate()
  n69 = Power(Base=OneMinus(None=ComponentMask(None=@n68) {r=true}), Exp=ScalarParameter() {defaultValue=3, parameterName=Lerp Power}) {constExponent=3}
  MaterialFunctionCall(Emissive=Multiply(A=LinearInterpolate(A=@n66.RGB, B=@n67.RGB, Alpha=@n69), B=@n19.RGB)) {materialFunction=&MF_FX_EmissiveAdjustments}
  Multiply(A=@n23.G)
  Multiply(A=@n37.R)
 ]
}
