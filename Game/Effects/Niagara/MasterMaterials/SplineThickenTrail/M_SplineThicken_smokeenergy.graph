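# Translucent, unlit Niagara master material for smoky energy trails on meshes, ribbons and sprites.
# Noise-distorted panning base texture, optional secondary texture and linear or dissolve masks drive a two-color gradient, optional UV tint and opacity with edge, camera, angle and depth fades.
# World position offset sums SplineThicken, particle-driven bend and panning-texture wobble.
# Features toggle via scalar lerps rather than static switches, while Niagara dynamic parameters scale UVs, speeds and offsets.
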Material /Game/Effects/Niagara/MasterMaterials/SplineThickenTrail/M_SplineThicken_smokeenergy
graph {
 properties {
  bAutomaticallySetUsageInEditor=false
  bUsedWithNiagaraMeshParticles=true
  bUsedWithNiagaraRibbons=true
  bUsedWithNiagaraSprites=true
  blendMode=BLEND_Translucent
  shadingModel=MSM_Unlit
  twoSided=true
  usedShadingModels=Unlit
 }
 nodes [
  n1 = ScalarParameter() {parameterName=OffsetUVs}
  n2 = Constant()
  n3 = Add(A=AppendVector(A=ScalarParameter() {parameterName=OffsetUVs}, B=Constant()), B=TextureCoordinate())
  n4 = VectorParameter() {
   group=NoiseTexture
   parameterName=NoiseUVControl
   sortPriority=2
   channelNames {a=SpeedY, b=SpeedX, g=ScaleY, r=ScaleX}
   defaultValue=<rgba 1 1 -0.5 0.25>
  }
  n5 = DynamicParameter() {paramNames [BaseTextureUV_Offset, BaseTextureSpeed_Offset, NoiseUV_Offset, NoiseSpeed_Offset]}
  n6 = Multiply(A=AppendVector(A=@n4.ScaleX, B=@n4.ScaleX), B=@n5.NoiseUV_Offset)
  n7 = AppendVector(A=@n4.SpeedX, B=@n4.SpeedX)
  n8 = TextureSampleParameter2D(UVs=Panner(Coordinate=Multiply(A=@n3, B=@n6) {constB=0.6}, Speed=Multiply(A=@n7, B=@n5.NoiseSpeed_Offset))) {
   group=NoiseTexture
   parameterName=NoiseTexture
   samplerType=SAMPLERTYPE_Grayscale
   sortPriority=1
   texture=/Game/Effects/Fort_Effects/Textures/Fire/T_WispyFireNoise_Soft_M
  }
  n9 = Add(
   A=Add(A=AppendVector(A=@n1, B=@n2), B=TextureCoordinate())
   B=Multiply(A=@n8.R, B=ScalarParameter() {defaultValue=0.2, group=NoiseTexture, parameterName=NoiseAmount, sortPriority=3}) {constB=0.2}
  )
  n10 = VectorParameter() {
   group=BaseTexture
   parameterName=BaseTextureUVControl
   sortPriority=2
   channelNames {a=SpeedY, b=SpeedX, g=ScaleY, r=ScaleX}
   defaultValue=<rgba 1 1 -0.5 0>
  }
  n11 = DynamicParameter() {paramNames [BaseTextureUV_Offset, BaseTextureSpeed_Offset, NoiseUV_Offset, NoiseSpeed_Offset]}
  n12 = AppendVector(A=Multiply(A=@n10.ScaleX, B=@n11.BaseTextureUV_Offset), B=@n10.ScaleY)
  n13 = Multiply(A=@n10.SpeedX, B=@n11.BaseTextureSpeed_Offset)
  n14 = Multiply(A=@n10.SpeedY, B=@n11.BaseTextureSpeed_Offset)
  n15 = TextureSampleParameter2D(UVs=Panner(Coordinate=Multiply(A=@n9, B=@n12), Speed=AppendVector(A=@n13, B=@n14)) {speedX=0.6}) {group=BaseTexture, parameterName=BaseTexture, sortPriority=1, texture=/Game/Effects/Fort_Effects/Textures/T_Chainmail_WispyMaskFat}
  n16 = ChannelMaskParameter(None=@n15.RGB) {group=MainTexture, parameterName=ChannelMask, sortPriority=5}
  n17 = VectorParameter() {parameterName=LerpTextureUV, defaultValue=<rgba 1.5 0.5 -1 0>}
  n18 = DynamicParameter() {paramNames [BaseTextureUV_Offset, BaseTextureSpeed_Offset, NoiseUV_Offset, NoiseSpeed_Offset]}
  n19 = Multiply(A=TextureCoordinate(), B=AppendVector(A=Multiply(A=@n17.R, B=@n18.BaseTextureUV_Offset), B=@n17.G))
  n20 = AppendVector(A=Multiply(A=@n17.B, B=@n18.BaseTextureSpeed_Offset), B=Multiply(A=@n17.A, B=@n18.BaseTextureSpeed_Offset))
  n21 = Multiply(
   A=@n16.R
   B=TextureSampleParameter2D(UVs=Panner(Coordinate=@n19, Speed=@n20) {speedX=-1.5}) {parameterName=LerpTexture, texture=/Game/Effects/Niagara/MasterMaterials/SplineThickenTrail/Texture/T_Pickaxe_SoapPocket_NoiseGlow}.R
  )
  n22 = ScalarParameter() {parameterName=Use_LerpTexture}
  n23 = TextureCoordinate()
  n24 = Multiply(A=ComponentMask(None=@n23) {g=true}) {constB=0.5}
  n25 = Multiply(A=OneMinus(None=ComponentMask(None=@n23) {r=true}), B=ScalarParameter() {defaultValue=1, parameterName=LinearMask_RightMul})
  n26 = ScalarParameter() {defaultValue=2, parameterName=LinearMask_RightPower}
  n27 = Multiply(A=Power(Base=Sine(None=@n24)), B=Saturate(None=Power(Base=@n25, Exp=@n26)))
  n28 = ScalarParameter() {defaultValue=1, parameterName=LinearMaskAmount}
  n29 = Saturate(None=Power(Base=Multiply(A=@n27, B=@n28), Exp=ScalarParameter() {defaultValue=1, parameterName=LinearMaskPower}))
  n30 = Add(
   A=TextureCoordinate() {uTiling=1.2}
   B=Multiply(A=TextureSample() {samplerType=SAMPLERTYPE_LinearColor, texture=/Game/Global/Textures/Default/Tileable_Patterns/T_PerlinNoise_Uncompressed}.R) {constB=0.05}
  )
  n31 = VectorParameter() {parameterName=MaskTextureUVControl, defaultValue=<rgba 1.5 1.5 -1.5 0.4>}
  n32 = DynamicParameter() {paramNames [BaseTextureUV_Offset, BaseTextureSpeed_Offset, NoiseUV_Offset, NoiseSpeed_Offset]}
  n33 = AppendVector(A=Multiply(A=@n31.R, B=@n32.BaseTextureUV_Offset), B=@n31.G)
  n34 = Multiply(A=@n31.B, B=@n32.BaseTextureSpeed_Offset)
  n35 = Multiply(A=@n31.A, B=@n32.BaseTextureSpeed_Offset)
  n36 = TextureSampleParameter2D(UVs=Panner(Coordinate=Multiply(A=@n30, B=@n33), Speed=AppendVector(A=@n34, B=@n35)) {speedX=-1}) {parameterName=MaskTexture, texture=/Game/CharClasses/Outlander/Gadgets/AntiMaterialCharge/Materials/T_AntiMaterialCharge_Dissolve_01}
  n37 = LinearInterpolate(A=ScalarParameter() {parameterName=MaskTexture_Min}, B=ScalarParameter() {defaultValue=-20, parameterName=MaskTexture_Max}, Alpha=@n36.R) {constA=-0.1}
  n38 = TextureCoordinate()
  n39 = Multiply(A=ComponentMask(None=@n38) {g=true}) {constB=0.5}
  n40 = Multiply(A=OneMinus(None=ComponentMask(None=@n38) {r=true})) {constB=3}
  n41 = Multiply(A=Power(Base=Sine(None=@n39)), B=Saturate(None=Power(Base=@n40)))
  n42 = Multiply(A=Power(Base=Multiply(A=@n41) {constB=1.5})) {constB=3}
  n43 = Add(
   A=Saturate(None=Add(A=@n37, B=@n42))
   B=SphereMask(A=TextureCoordinate() {uTiling=0.35, vTiling=0.6}, B=Constant2Vector() {g=0.3}) {attenuationRadius=0.2, hardnessPercent=50}
  )
  n44 = ScalarParameter() {parameterName=Use_MaskTexture}
  n45 = Subtract(
   A=Multiply(A=LinearInterpolate(A=@n16.R, B=@n21, Alpha=@n22), B=LinearInterpolate(A=@n29, B=@n43, Alpha=@n44))
   B=ScalarParameter() {parameterName=TextureSubtract}
  ) {constB=0.5}
  n46 = Saturate(None=@n45)
  n47 = ScalarParameter() {defaultValue=3, parameterName=ColorProportion}
  n48 = ParticleColor()
  n49 = Multiply(
   A=LinearInterpolate(
    A=VectorParameter() {group=Color, parameterName=DarkColor, sortPriority=1, defaultValue=<rgba 0.02 0.02 0.02 1>}.RGB
    B=VectorParameter() {group=Color, parameterName=LightColor, sortPriority=2, defaultValue=<rgba 0.5 0.5 0.5 1>}.RGB
    Alpha=Power(Base=@n46, Exp=@n47)
   )
   B=@n48.RGB
  )
  n50 = ComponentMask(None=TextureCoordinate()) {r=true}
  n51 = LinearInterpolate(
   A=VectorParameter() {parameterName=LinearColorA, defaultValue=<rgba 0 0.429 1 1>}.RGB
   B=VectorParameter() {parameterName=LinearColorB, defaultValue=<rgba 0.0677 1 0.45 1>}.RGB
   Alpha=Power(Base=OneMinus(None=@n50), Exp=ScalarParameter() {defaultValue=1.2, parameterName=LinearColor_MaskPower}) {constExponent=1.2}
  )
  n52 = ScalarParameter() {parameterName=LinearColor_Des}
  n53 = LinearInterpolate(A=@n49, B=Multiply(A=Desaturation(None=@n51, Fraction=@n52), B=@n49), Alpha=ScalarParameter() {parameterName=Use_LinearColor})
  n54 = ScalarParameter() {defaultValue=0.15, group=Color, parameterName=Desaturation, sortPriority=3}
  MaterialFunctionCall(Emissive=Multiply(A=Desaturation(None=@n53, Fraction=@n54), B=ScalarParameter() {defaultValue=1, parameterName=All_Light})) {materialFunction=&MF_FX_EmissiveAdjustments}
  n55 = DynamicParameter() {parameterIndex=3, paramNames [SplineBase_Offset, SplineTip_Offset, WPOAmount, WPO_Offset]}
  n56 = Multiply(A=ScalarParameter() {defaultValue=1, parameterName=Spline_Base}, B=@n55.SplineBase_Offset)
  n57 = Multiply(A=ScalarParameter() {defaultValue=1, parameterName=Spline_Tip}, B=@n55.SplineTip_Offset)
  n58 = StaticBool()
  n59 = LinearInterpolate(
   A=Constant()
   B=Multiply(
    A=ScalarParameter() {defaultValue=1, parameterName=Spline_ThickMul}
    B=MaterialFunctionCall(WidthBase=@n56, WidthTip=@n57, Expand U or V UV Channel=@n58) {materialFunction=&SplineThicken}.WorldPositionOffset
   )
   Alpha=ScalarParameter() {parameterName=Use_SplineThicken}
  )
  n60 = DynamicParameter() {desc=Parameters set by Niagara, parameterIndex=1, paramNames [X,Y,Z,WPOIntensity]}
  n61 = DynamicParameter() {desc=Parameters set by Niagara, parameterIndex=2, paramNames [X, Y, Z, ""]}
  n62 = Subtract(
   A=MaterialFunctionCall(X=@n60.X, Y=@n60.X, Z=@n60.X) {materialFunction=&MakeFloat3}.Result
   B=LinearInterpolate(
    A=ParticlePositionWS()
    B=MaterialFunctionCall(X=@n61.X, Y=@n61.X, Z=@n61.X) {materialFunction=&MakeFloat3}.Result
    Alpha=ScalarParameter() {parameterName=FEPositionFix}
   )
  )
  n63 = ScalarParameter() {defaultValue=50, parameterName=Interpolate_WPO_Range}
  n64 = Multiply(A=@n63) {constB=-1}
  n65 = Power(Base=ComponentMask(None=TextureCoordinate()) {r=true}, Exp=ScalarParameter() {defaultValue=1.5, parameterName=WPOBendPower})
  n66 = LinearInterpolate(
   A=Constant()
   B=LinearInterpolate(B=Clamp(None=@n62, Min=@n64, Max=@n63) {maxDefault=50, minDefault=-50}, Alpha=Multiply(A=@n65, B=@n60.WPOIntensity))
   Alpha=ScalarParameter() {parameterName=Use interpolate WPO}
  )
  n67 = Constant()
  n68 = TextureCoordinate() {vTiling=0}
  n69 = DynamicParameter() {parameterIndex=3, paramNames [SplineBase_Offset, SplineTip_Offset, WPOAmount, WPO_Offset]}
  n70 = AppendVector(
   A=Multiply(A=ScalarParameter() {defaultValue=1, parameterName=WPOTextureU}, B=@n69.WPO_Offset)
   B=Multiply(A=ScalarParameter() {parameterName=WPOTextureV}, B=@n69.WPO_Offset)
  )
  n71 = Multiply(A=ScalarParameter() {defaultValue=-1, parameterName=WPOTextureSpeedX}, B=@n69.WPO_Offset)
  n72 = Multiply(A=ScalarParameter() {parameterName=WPOTextureSpeedY}, B=@n69.WPO_Offset)
  n73 = TextureSampleParameter2D(UVs=Panner(Coordinate=Multiply(A=@n68, B=@n70), Speed=AppendVector(A=@n71, B=@n72)) {speedX=-2}) {parameterName=WPOTexture, samplerType=SAMPLERTYPE_Normal, texture=&T_Weap_Distort_N_02}
  n74 = TextureCoordinate()
  n75 = Multiply(
   A=LinearInterpolate(Alpha=@n73.R) {constB=1.5}
   B=LinearInterpolate(A=ScalarParameter() {defaultValue=0.05, parameterName=WPOMask}, Alpha=ComponentMask(None=@n74) {r=true}) {constA=0.05}
  )
  n76 = LinearInterpolate(A=@n75, B=Multiply(A=VertexNormalWS(), B=@n75), Alpha=ScalarParameter() {parameterName=Use_VertexNormalMul})
  n77 = Multiply(A=Transform(None=Constant3Vector() {constant=<rgba 1 0 0 0>}) {transformSourceType=TRANSFORMSOURCE_Local}, B=@n76)
  n78 = DynamicParameter() {parameterIndex=3, paramNames [SplineBase_Offset, SplineTip_Offset, WPOAmount, WPO_Offset]}
  n79 = Multiply(A=ScalarParameter() {defaultValue=1, parameterName=X WPO_Intensity}, B=@n78.WPOAmount)
  n80 = Multiply(A=Transform(None=Constant3Vector() {constant=<rgba 0 1 0 0>}) {transformSourceType=TRANSFORMSOURCE_Local}, B=@n76)
  n81 = Multiply(A=ScalarParameter() {defaultValue=1, parameterName=Y WPO_Intensity}, B=@n78.WPOAmount)
  n82 = Transform(None=Constant3Vector() {constant=<rgba 0 0 1 0>}) {transformSourceType=TRANSFORMSOURCE_Local}
  n83 = ScalarParameter() {defaultValue=1, parameterName=Z WPO_Intensity}
  n84 = Add(
   A=Add(A=Multiply(A=@n77, B=@n79) {constB=10}, B=Multiply(A=@n80, B=@n81) {constB=10})
   B=Multiply(A=Multiply(A=@n82, B=@n76), B=Multiply(A=@n83, B=@n78.WPOAmount)) {constB=10}
  )
  n85 = ScalarParameter() {parameterName=Use_WPO}
  Reroute(None=Add(A=Add(A=@n59, B=@n66), B=LinearInterpolate(A=@n67, B=@n84, Alpha=@n85)))
  n86 = Multiply(A=@n48.A, B=Multiply(A=@n45, B=ScalarParameter() {defaultValue=3, group=Color, parameterName=All_Alpha, sortPriority=4}) {constB=3})
  n87 = SmoothStep(Max=ScalarParameter() {defaultValue=0.2, parameterName=LeftEdgeMaskMax}, Value=ComponentMask(None=TextureCoordinate()) {r=true})
  n88 = ScalarParameter() {parameterName=NearFade, sortPriority=7}
  n89 = Multiply(
   A=Multiply(
    A=Multiply(A=@n86, B=@n87)
    B=MaterialFunctionCall(Distance=@n88) {materialFunction=&Near_Camera_Fade}.Result
   )
   B=MaterialFunctionCall("FadeSensitivity (optional)"=ScalarParameter() {parameterName=MeshFade, sortPriority=6}) {materialFunction=&FadeBasedOnViewAngle}.Result
  )
  n90 = DepthFade(Opacity=@n89, FadeDistance=ScalarParameter() {defaultValue=1, parameterName=DepthFade, sortPriority=5})
  ShadingPathSwitch(Default=@n90, Deferred=@n90, Forward=@n89, Mobile=@n89)
 ]
}
